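# Parent material for the TriadSpear backpack skin, layering audio-reactive equalizer bar and red/blue line glow effects onto the 2023 character parent PreFX/PostFX functions.
# Frontend-loop audio amplitude drives the effects, with panning noise substituting when there is no signal.
# Cylindrically projected noise distorts line UVs, while sine-modulated offsets shift bar mask samples.
# Amplitude also pushes a sphere-masked UV region along local -Y, and a temporal responsiveness output is written.
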Material /BRCosmetics/Accessories/FORT_Backpacks/M_MED_80sExercise/Skins/TriadSpear/Materials/M_TriadSpear_Backpack_Parent
graph {
 properties {
  bAutomaticallySetUsageInEditor=false
  bUseMaterialAttributes=true
  bUsedWithMeshDeformer=true
  bUsedWithMorphTargets=true
  bUsedWithSkeletalMesh=true
  pixelDepthOffsetMode=PDOM_Legacy
  shadingModel=MSM_FromMaterialExpression
  usedShadingModels=Default Lit | Subsurface | Clear Coat | Subsurface Profile | Cloth
 }
 nodes [
  n1 = MaterialFunctionCall() {materialFunction=&MF_CharacterParent_2023_PreFX}
  n2 = GetMaterialAttributes(None=@n1.Result) {attributeGetTypes [#1]}
  n3 = NamedRerouteUsage()
  n4 = NamedRerouteUsage()
  n5 = Transform(None=Constant4Vector() {constant=<rgba 0 -1 0 1>}.RGB) {transformSourceType=TRANSFORMSOURCE_Local}
  n6 = Constant4Vector() {constant=<rgba 0.752 0.513 0 1>}
  n7 = MaterialFunctionCall!(InAmplitude, InSpectrum, CustomFakeSpectrum, CustomFakeAmplitude) {materialFunction=/BRCosmetics/Global/MaterialFunctions/AudioReactivity/MF_AudioReactivity_FrontendLoop}
  n8 = Constant() {r=3}
  n9 = Multiply(
   A=SphereMask(A=TextureCoordinate(), B=ComponentMask(None=@n6.RGB) {g=true, r=true}, Radius=Constant() {r=0.07}, Hardness=Constant() {r=0.2})
   B=Multiply(A=Constant() {r=1.8}, B=Power(Base=@n7.Amplitude, Exp=@n8) {constExponent=4})
  )
  MaterialFunctionCall(
   Material=SetMaterialAttributes!(MaterialAttributes=@n2.MaterialAttributes, Emissive Color=Add(A=@n3, B=@n4), World Position Offset=Multiply(A=@n5, B=@n9)) {attributeSetTypes [#1, #2]}
   M=@n1.M
   Pre Skinned Local Position=@n1.M
   Distance Blend=@n1.M
   Fresnel=@n1.M
  ) {materialFunction=&MF_CharacterParent_2023_PostFX}
  n10 = NamedRerouteUsage()
  n11 = MaterialFunctionCall!(Tile=Constant4Vector() {constant=<rgba 1 1 0 1>}.RGB, DistortionAdjustment=Constant() {r=-7.1828}) {materialFunction=/BRCosmetics/Characters/Player/Female/Medium/Bodies/F_MED_TriadSpear/Materials/Body/MF_CylindricalProjUV}
  n12 = TextureSample!(UVs=Panner(Coordinate=@n11.UVs) {speedX=0.1, speedY=0.2}, Tex, "DDX(UVs)"=@n11.DDX, "DDY(UVs)"=@n11.DDX, Apply View MipBias) {
   mipValueMode=TMVM_Derivative
   samplerSource=SSM_Wrap_WorldGroupSettings
   samplerType=SAMPLERTYPE_Grayscale
   texture=/Game/Global/Textures/Default/Noise/LowResBlurredNoiseNormalized
  }
  n13 = LinearInterpolate(A=Multiply(A=Subtract(A=@n12.R) {constB=0.5}) {constB=2}, Alpha=NamedRerouteUsage())
  n14 = Multiply(A=@n13, B=Constant() {r=1})
  NamedRerouteDeclaration(None=Add(A=TextureCoordinate(), B=Multiply(A=@n10, B=@n14) {constB=1.3})) {name=RedLineUV1, nodeColor=<rgba 0.6641 0 1 1>}
  n15 = TextureSample!(UVs=NamedRerouteUsage(), Tex, Apply View MipBias) {
   samplerSource=SSM_Wrap_WorldGroupSettings
   samplerType=SAMPLERTYPE_Masks
   texture=/BRCosmetics/Accessories/FORT_Backpacks/M_MED_80sExercise/Skins/TriadSpear/Textures/T_TriadSpear_Backpack_FX_redLine
  }
  n16 = NamedRerouteUsage()
  n17 = TextureSample(UVs=@n16) {
   samplerSource=SSM_Wrap_WorldGroupSettings
   samplerType=SAMPLERTYPE_Masks
   texture=/BRCosmetics/Accessories/FORT_Backpacks/M_MED_80sExercise/Skins/TriadSpear/Textures/T_TriadSpear_Backpack_FX_blueLine
  }
  n18 = TextureSample(UVs=NamedRerouteUsage()) {
   samplerSource=SSM_Wrap_WorldGroupSettings
   samplerType=SAMPLERTYPE_Masks
   texture=&T_TriadSpear_Backpack_FX_blueLine
  }
  NamedRerouteDeclaration(None=Max(A=Max(A=@n15.B, B=@n17.G), B=Max(A=@n18.B, B=@n17.B))) {name=LineMask, nodeColor=<rgba 0.3828 1 0 1>}
  n19 = Constant()
  NamedRerouteDeclaration(None=If(A=NamedRerouteUsage(), "A > B"=Constant() {r=1}, "A == B"=@n19, "A < B"=@n19)) {name=In_Ampltiude_LineSwitch, nodeColor=<rgba 1 0.2812 0 1>}
  n20 = NamedRerouteUsage()
  n21 = MaterialFunctionCall(Tile=Constant4Vector() {constant=<rgba 1 1 0 1>}.RGB, DistortionAdjustment=Constant() {r=-7.1828}) {materialFunction=&MF_CylindricalProjUV}
  n22 = Constant4Vector() {constant=<rgba 0.05 0.1 0 1>}
  n23 = TextureSample!(
   UVs=Panner(Coordinate=@n21.UVs, Speed=ComponentMask(None=@n22.RGB) {g=true, r=true}) {speedX=0.1, speedY=0.2}
   Tex
   "DDX(UVs)"=@n21.DDX
   "DDY(UVs)"=@n21.DDX
   Apply View MipBias
  ) {
   mipValueMode=TMVM_Derivative
   samplerSource=SSM_Wrap_WorldGroupSettings
   samplerType=SAMPLERTYPE_Grayscale
   texture=&LowResBlurredNoiseNormalized
  }
  n24 = LinearInterpolate(A=Multiply(A=Subtract(A=@n23.R) {constB=0.5}) {constB=2}, Alpha=NamedRerouteUsage())
  n25 = Multiply(A=@n24, B=Constant() {r=-1.5})
  NamedRerouteDeclaration(None=Add(A=TextureCoordinate(), B=Multiply(A=@n20, B=@n25) {constB=1.3})) {name=RedLineUV2, nodeColor=<rgba 0.6641 0 1 1>}
  n26 = TextureObjectParameter() {
   group=TriadSpear_backpack
   parameterName=Bar_Mask
   samplerType=SAMPLERTYPE_Masks
   texture=/BRCosmetics/Accessories/FORT_Backpacks/M_MED_80sExercise/Skins/TriadSpear/Textures/T_TriadSpear_Backpack_Bar_E
  }
  n27 = TextureSample!(UVs, Tex=@n26, Apply View MipBias) {
   samplerType=SAMPLERTYPE_Masks
   texture=&T_TriadSpear_Backpack_Bar_E
  }
  n28 = NamedRerouteDeclaration(None=@n27.B) {name=Bar_Glow_Power, nodeColor=<rgba 1 0 0.6797 1>}
  n29 = LinearInterpolate(B=Constant() {r=0.007}, Alpha=Multiply(A=OneMinus(None=@n28), B=NamedRerouteUsage())) {constB=0.03}
  n30 = Constant() {r=2}
  n31 = NamedRerouteUsage()
  n32 = TextureCoordinate()
  n33 = Sine(None=Add(A=LinearInterpolate(B=@n30, Alpha=@n31) {constB=5}, B=ComponentMask(None=@n32) {r=true}))
  n34 = LinearInterpolate(Alpha=Multiply(A=Power(Base=@n33)) {constB=1.5}) {constA=0.5}
  n35 = AppendVector(A=Constant(), B=LinearInterpolate(A=@n29, B=Multiply(A=@n29, B=@n34), Alpha=NamedRerouteUsage()))
  n36 = TextureObjectParameter() {
   group=TriadSpear_backpack
   parameterName=Bar_Mask
   samplerType=SAMPLERTYPE_Masks
   texture=&T_TriadSpear_Backpack_Bar_E
  }
  n37 = TextureSample(UVs=Add(A=TextureCoordinate(), B=Multiply(A=@n35) {constB=3}), Tex=@n36) {
   samplerType=SAMPLERTYPE_Masks
   texture=&T_TriadSpear_Backpack_Bar_E
  }
  n38 = TextureSample(UVs=Add(A=TextureCoordinate(), B=Multiply(A=@n35) {constB=2}), Tex=@n36) {
   samplerType=SAMPLERTYPE_Masks
   texture=&T_TriadSpear_Backpack_Bar_E
  }
  n39 = Add(A=@n37.G, B=@n38.R)
  n40 = Multiply(
   A=LinearInterpolate(A=Constant3Vector() {constant=<rgba 0.0056 0.0343 0.8632 1>}, B=Constant3Vector() {constant=<rgba 0.051 0.3031 0.9387 1>}, Alpha=@n39)
   B=@n39
  )
  n41 = TextureObjectParameter() {
   parameterName=Bar_BaseColor
   texture=/BRCosmetics/Accessories/FORT_Backpacks/M_MED_80sExercise/Skins/TriadSpear/Textures/T_TriadSpear_Backpack_E2
  }
  n42 = TextureSample(UVs=Add(A=TextureCoordinate(), B=Multiply(A=@n35) {constB=0.5}), Tex=@n41) {texture=&T_TriadSpear_Backpack_E2}
  n43 = NamedRerouteUsage()
  NamedRerouteDeclaration(None=Multiply(A=Add(A=@n40, B=@n42.RGB), B=OneMinus(None=@n43))) {name=BarEmissiveColor, nodeColor=<rgba 0 1 0.0625 1>}
  n44 = TextureSample(UVs=NamedRerouteUsage()) {
   samplerSource=SSM_Wrap_WorldGroupSettings
   samplerType=SAMPLERTYPE_Masks
   texture=&T_TriadSpear_Backpack_FX_redLine
  }
  NamedRerouteDeclaration(None=Max(A=@n15.R, B=@n44.G)) {name=Redline, nodeColor=<rgba 0.1953 0 1 1>}
  n45 = TextureSample(UVs=@n16) {
   samplerSource=SSM_Wrap_WorldGroupSettings
   samplerType=SAMPLERTYPE_Masks
   texture=&T_TriadSpear_Backpack_FX_redLine
  }
  NamedRerouteDeclaration(None=Max(A=@n18.R, B=@n45.G)) {name=BlueLine, nodeColor=<rgba 0.125 1 0 1>}
  NamedRerouteDeclaration(None=@n2.Emissive Color) {name=Input_EmissvieColor, nodeColor=<rgba 1 0 0 1>}
  n46 = NamedRerouteUsage()
  n47 = Multiply(A=@n24, B=Constant() {r=1})
  NamedRerouteDeclaration(None=Add(A=TextureCoordinate(), B=Multiply(A=@n46, B=@n47))) {name=BluelineUV2, nodeColor=<rgba 0 1 0.3438 1>}
  n48 = Constant() {r=0.5}
  n49 = MaterialFunctionCall(Tile=Constant() {r=1}, DistortionAdjustment=Constant() {r=-7.1828}) {materialFunction=&MF_CylindricalProjUV}
  n50 = TextureSample!(UVs=Panner(Coordinate=@n49.UVs) {speedY=0.2}, Tex, "DDX(UVs)"=@n49.DDX, "DDY(UVs)"=@n49.DDX, Apply View MipBias) {
   mipValueMode=TMVM_Derivative
   samplerSource=SSM_Wrap_WorldGroupSettings
   samplerType=SAMPLERTYPE_Grayscale
   texture=&LowResBlurredNoiseNormalized
  }
  n51 = Saturate(None=Power(Base=Multiply(A=@n50.R) {constB=2}) {constExponent=3})
  n52 = Constant3Vector() {constant=<rgba 0 0.1701 1 1>}
  n53 = Constant3Vector() {constant=<rgba 0 0.9759 1.5 1>}
  n54 = MaterialFunctionCall(Tile=Constant() {r=1}, DistortionAdjustment=Constant() {r=-7.1828}) {materialFunction=&MF_CylindricalProjUV}
  n55 = TextureSample(UVs=@n54.UVs, "DDX(UVs)"=@n54.UVs, "DDY(UVs)"=@n54.UVs) {
   mipValueMode=TMVM_Derivative
   samplerSource=SSM_Wrap_WorldGroupSettings
   samplerType=SAMPLERTYPE_Grayscale
   texture=&LowResBlurredNoiseNormalized
  }
  n56 = Constant3Vector() {constant=<rgba 0.8385 0 0.0037 1>}
  n57 = Constant3Vector() {constant=<rgba 1 0.0168 0 1>}
  n58 = LinearInterpolate(
   A=Multiply(A=NamedRerouteUsage(), B=LinearInterpolate(A=@n52, B=@n53, Alpha=@n55.R))
   B=Multiply(A=NamedRerouteUsage(), B=LinearInterpolate(A=@n56, B=@n57, Alpha=@n55.R))
   Alpha=NamedRerouteUsage()
  )
  n59 = NamedRerouteUsage()
  n60 = Constant() {r=1}
  n61 = LinearInterpolate(A=If(A=@n59, "A > B"=Constant() {r=1}, "A == B"=@n60, "A < B"=@n60), B=Constant() {r=3}, Alpha=@n59) {constB=10}
  NamedRerouteDeclaration(None=Multiply(A=Multiply(A=@n58, B=@n61), B=LinearInterpolate(A=@n48, Alpha=@n51) {constA=0.7})) {name=EqualizerRedBlueLine, nodeColor=<rgba 0 0.5312 1 1>}
  n62 = NamedRerouteUsage()
  n63 = Power(Base=@n62, Exp=Constant() {r=2}) {constExponent=4}
  n64 = MaterialFunctionCall(Tile=Constant() {r=1}, DistortionAdjustment=Constant() {r=-7.1828}) {materialFunction=&MF_CylindricalProjUV}
  n65 = TextureSample(UVs=Panner(Coordinate=@n64.UVs) {speedY=0.05}, "DDX(UVs)"=@n64.DDX, "DDY(UVs)"=@n64.DDX) {
   mipValueMode=TMVM_Derivative
   samplerSource=SSM_Wrap_WorldGroupSettings
   samplerType=SAMPLERTYPE_Grayscale
   texture=&LowResBlurredNoiseNormalized
  }
  n66 = Multiply(A=LinearInterpolate(Alpha=@n65.R) {constA=0.5}, B=Constant() {r=0.2})
  NamedRerouteDeclaration(None=If(A=@n62, "A > B"=LinearInterpolate(Alpha=@n63) {constB=0.2}, "A == B"=@n66, "A < B"=@n66)) {name=In_LineAmpltiude, nodeColor=<rgba 1 0.2812 0 1>}
  n67 = NamedRerouteUsage()
  n68 = Constant() {r=3}
  n69 = Constant() {r=0.1}
  n70 = Constant() {r=-7.1828}
  n71 = TextureSample!(
   UVs=Panner(Coordinate=MaterialFunctionCall(Tile=@n69, DistortionAdjustment=@n70) {materialFunction=&MF_CylindricalProjUV}.UVs) {speedX=0.2, speedY=0.7}
   Tex
   Apply View MipBias
  ) {
   samplerSource=SSM_Wrap_WorldGroupSettings
   samplerType=SAMPLERTYPE_Grayscale
   texture=&LowResBlurredNoiseNormalized
  }
  n72 = Multiply(A=@n71.R, B=Constant() {r=1.5})
  NamedRerouteDeclaration(None=If(A=@n67, "A > B"=Power(Base=@n67, Exp=@n68) {constExponent=4}, "A == B"=@n72, "A < B"=@n72)) {name=IN_Amplitude_Bar, nodeColor=<rgba 0 0.6484 1 1>}
  n73 = Constant()
  NamedRerouteDeclaration(None=If(A=NamedRerouteUsage(), "A > B"=Constant() {r=1}, "A == B"=@n73, "A < B"=@n73)) {name=AudioSwitchBar, nodeColor=<rgba 0.4766 1 0 1>}
  NamedRerouteDeclaration(None=MaterialFunctionCall() {materialFunction=&MF_AudioReactivity_FrontendLoop}.Amplitude) {name=SourceAmplitude, nodeColor=<rgba 0.1016 0 1 1>}
  n74 = MaterialFunctionCall(
   UVs=MaterialFunctionCall(Tile=Constant() {r=2}, DistortionAdjustment=Constant() {r=-2}) {materialFunction=&MF_CylindricalProjUV}.UVs
   "Rotation Angle (0-1)"=Constant() {r=0.22}
  ) {materialFunction=&CustomRotator}
  n75 = ComponentMask(None=Constant4Vector() {constant=<rgba 0 0.15 0 1>}.RGB) {g=true, r=true}
  n76 = LinearInterpolate(A=@n75, B=Multiply(A=@n75) {constB=0.25}, Alpha=NamedRerouteUsage())
  n77 = ComponentMask(None=Multiply(A=Panner(Coordinate=@n74.Rotated Values, Speed=@n76) {speedY=0.03}, B=Constant() {r=3.5})) {g=true}
  n78 = Saturate(None=MaterialFunctionCall(
    Input=Constant() {r=0.36}
    Input Low=ScalarParameter() {defaultValue=0.25, group=TopSide_V02_Layer Emissive_01 Reactive, parameterName=TopS Audio 01_Low_Remap, sortPriority=149}
    Input High=ScalarParameter() {defaultValue=1, group=TopSide_V02_Layer Emissive_01 Reactive, parameterName=TopS Audio 01_High_Remap, sortPriority=149}
    Target Low=Constant()
    Target High=Constant() {r=1}
   ) {materialFunction=&RemapValueRange}.Result)
  n79 = ScalarParameter() {defaultValue=0.6, group=TopSide_V02_Layer Emissive_01 Reactive, parameterName=TopS Audio 01_Falloff, sortPriority=149}
  n80 = Multiply(A=Multiply(A=Sine(None=@n77), B=Power(Base=@n78, Exp=@n79)), B=Constant() {r=-0.1131})
  n81 = Constant()
  n82 = AppendVector(A=Constant(), B=Constant() {r=3})
  n83 = Constant()
  n84 = Multiply(
   A=LinearInterpolate(A=Constant() {r=-0.1}, B=Constant() {r=1.1}, Alpha=NamedRerouteUsage()) {constA=-0.05, constB=0.05}
   B=If(A=NamedRerouteUsage(), "A > B"=Constant() {r=1}, "A == B"=@n83, "A < B"=@n83)
  )
  NamedRerouteDeclaration(None=Multiply(A=Add(A=@n80, B=@n81), B=Multiply(A=@n82, B=@n84))) {name=AudioLineUV, nodeColor=<rgba 0 1 0.8828 1>}
  n85 = Constant() {r=0.25}
  n86 = NamedRerouteUsage()
  n87 = NamedRerouteUsage()
  n88 = Multiply(
   A=NamedRerouteUsage()
   B=LinearInterpolate(B=Constant() {r=1}, Alpha=LinearInterpolate(A=@n85, B=@n86, Alpha=@n87) {constA=1}) {constA=0.1, constB=10}
  )
  n89 = NamedRerouteUsage()
  n90 = Multiply(A=@n89) {constB=0.2}
  n91 = NamedRerouteUsage()
  n92 = SmoothStep(
   Min=Constant()
   Max=LinearInterpolate(A=Constant() {r=0.25}, B=Constant() {r=0.001}, Alpha=Saturate(None=@n86)) {constA=5, constB=0}
   Value=Saturate(None=Add(A=@n91) {constB=0.058})
  )
  n93 = Saturate(None=Multiply(A=@n89, B=LinearInterpolate(Alpha=@n92) {constA=0.1, constB=1.3}))
  n94 = LinearInterpolate(Alpha=NamedRerouteUsage()) {constB=2}
  NamedRerouteDeclaration(None=Add(A=Add(A=@n88, B=@n90), B=Multiply(A=@n93, B=@n94))) {name=EqualizerBar, nodeColor=<rgba 0.8516 1 0 1>}
  n95 = NamedRerouteUsage()
  n96 = Multiply(A=@n13, B=Constant() {r=-1.5})
  NamedRerouteDeclaration(None=Add(A=TextureCoordinate(), B=Multiply(A=@n95, B=@n96))) {name=BluelineUV1, nodeColor=<rgba 0 1 0.3438 1>}
  n97 = NamedRerouteUsage()
  n98 = Add(A=NamedRerouteUsage(), B=Multiply(A=ComponentMask(None=@n97) {b=true}) {constB=2})
  TemporalResponsivenessOutput!(None=LinearInterpolate(Alpha=Saturate(None=@n98)) {constA=0.4})
 ]
}
